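# Anim Blueprint for the PinkySight backpack cosmetic, playing reactive animations to player events.
# The Backpack States machine branches from looping Idle to DBNO, elimination, health/shield gain and chest-open reactions, returning afterward.
# Anim input events set the trigger flags, cleared post-evaluate or, for health/shield, by a 0.2s retriggerable delay.
# An unconnected Animation Testing machine chains the clips; movement and Skydive values are computed but unused in these graphs.
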AnimBlueprint /BRCosmetics/Accessories/FORT_Backpacks/Backpack_M_MED_PinkySight/Meshes/M_MED_PinkySight_Pack_AnimBP
target_skeleton=&/M_MED_PinkySight_Pack_Skeleton
graphs [
 AnimGraph {nodes [
   Input Pose()
   AnimGraphNode_StateMachine_1 = Backpack States() {is=title AnimGraphNode_StateMachine}
   Output Pose(Result=@AnimGraphNode_StateMachine_1) {layerGroup=None}
   AnimGraphNode_StateMachine_0 = Animation Testing() {is=title AnimGraphNode_StateMachine}
  ]}
 Animation Testing {
  entry=Idle_first
  path=AnimGraph.AnimGraphNode_StateMachine_0.Animation Testing
  ids=id
  nodes [
   AnimStateNode_0 = Idle_first() {is=title AnimStateNode}
   AnimStateNode_1 = Idle_DBNO_Repeat() {is=title AnimStateNode}
   AnimStateNode_2 = Idle_second() {is=title AnimStateNode}
   AnimStateNode_3 = Reactive_GainShield() {is=title AnimStateNode}
   AnimStateNode_4 = Reactive_Elim() {is=title AnimStateNode}
   AnimStateNode_5 = Reactive_ChestOpen() {is=title AnimStateNode}
   AnimStateNode_6 = Idle_DBNO() {is=title AnimStateNode}
   AnimStateTransitionNode_0 = AnimStateTransitionNode() {bAutomaticRuleBasedOnSequencePlayerInState=true, crossfadeDuration=0}
   AnimStateTransitionNode_1 = AnimStateTransitionNode() {bAllowInertializationForSelfTransitions=true, bAutomaticRuleBasedOnSequencePlayerInState=true}
   AnimStateTransitionNode_2 = AnimStateTransitionNode() {bAutomaticRuleBasedOnSequencePlayerInState=true}
   AnimStateTransitionNode_3 = AnimStateTransitionNode() {bAutomaticRuleBasedOnSequencePlayerInState=true}
   AnimStateTransitionNode_4 = AnimStateTransitionNode() {bAutomaticRuleBasedOnSequencePlayerInState=true}
   AnimStateTransitionNode_5 = AnimStateTransitionNode() {bAutomaticRuleBasedOnSequencePlayerInState=true}
   AnimStateTransitionNode_7 = AnimStateTransitionNode() {bAllowInertializationForSelfTransitions=true, bAutomaticRuleBasedOnSequencePlayerInState=true}
  ]
  links [
   {
    dst=Idle_DBNO_Repeat
    rule=AnimGraph.AnimGraphNode_StateMachine_0.Animation Testing.AnimStateTransitionNode_1.Transition
    src=Idle_DBNO
    via=AnimStateTransitionNode_1
   }
   {
    dst=Idle_first
    rule=AnimGraph.AnimGraphNode_StateMachine_0.Animation Testing.AnimStateTransitionNode_7.Transition
    src=Idle_DBNO_Repeat
    via=AnimStateTransitionNode_7
   }
   {
    dst=Idle_second
    rule=AnimGraph.AnimGraphNode_StateMachine_0.Animation Testing.AnimStateTransitionNode_0.Transition
    src=Idle_first
    via=AnimStateTransitionNode_0
   }
   {
    dst=Reactive_GainShield
    rule=AnimGraph.AnimGraphNode_StateMachine_0.Animation Testing.AnimStateTransitionNode_2.Transition
    src=Idle_second
    via=AnimStateTransitionNode_2
   }
   {
    dst=Idle_DBNO
    rule=AnimGraph.AnimGraphNode_StateMachine_0.Animation Testing.AnimStateTransitionNode_5.Transition
    src=Reactive_ChestOpen
    via=AnimStateTransitionNode_5
   }
   {
    dst=Reactive_ChestOpen
    rule=AnimGraph.AnimGraphNode_StateMachine_0.Animation Testing.AnimStateTransitionNode_4.Transition
    src=Reactive_Elim
    via=AnimStateTransitionNode_4
   }
   {
    dst=Reactive_Elim
    rule=AnimGraph.AnimGraphNode_StateMachine_0.Animation Testing.AnimStateTransitionNode_3.Transition
    src=Reactive_GainShield
    via=AnimStateTransitionNode_3
   }
  ]
 }
 Idle_first {
  path=AnimGraph.AnimGraphNode_StateMachine_0.Animation Testing.AnimStateNode_0.Idle_first
  nodes [Output Animation Pose(Result=Sequence Player 'PinkySight_Idle_PACK'() {bLoopAnimation=false, sequence=/BRCosmetics/Accessories/FORT_Backpacks/Backpack_M_MED_PinkySight/Animations/PinkySight_Idle_PACK})]
 }
 Idle_DBNO_Repeat {
  path=AnimGraph.AnimGraphNode_StateMachine_0.Animation Testing.AnimStateNode_1.Idle_DBNO_Repeat
  nodes [
   n1 = Sequence Player 'PinkySight_Reactive_Sad_PACK'() {bLoopAnimation=false, sequence=/BRCosmetics/Accessories/FORT_Backpacks/Backpack_M_MED_PinkySight/Animations/PinkySight_Reactive_Sad_PACK}
   Output Animation Pose(Result=@n1)
  ]
 }
 Idle_second {
  path=AnimGraph.AnimGraphNode_StateMachine_0.Animation Testing.AnimStateNode_2.Idle_second
  nodes [Output Animation Pose(Result=Sequence Player 'PinkySight_Idle_PACK'() {bLoopAnimation=false, sequence=&PinkySight_Idle_PACK})]
 }
 Reactive_GainShield {
  path=AnimGraph.AnimGraphNode_StateMachine_0.Animation Testing.AnimStateNode_3.Reactive_GainShield
  nodes [
   n1 = Sequence Player 'PinkySight_Reactive_Celebrate_PACK'() {
    bLoopAnimation=false
    sequence=/BRCosmetics/Accessories/FORT_Backpacks/Backpack_M_MED_PinkySight/Animations/PinkySight_Reactive_Celebrate_PACK
   }
   Output Animation Pose(Result=@n1)
  ]
 }
 Reactive_Elim {
  path=AnimGraph.AnimGraphNode_StateMachine_0.Animation Testing.AnimStateNode_4.Reactive_Elim
  nodes [
   n1 = Sequence Player 'PinkySight_Reactive_HopHop_PACK'() {bLoopAnimation=false, sequence=/BRCosmetics/Accessories/FORT_Backpacks/Backpack_M_MED_PinkySight/Animations/PinkySight_Reactive_HopHop_PACK}
   Output Animation Pose(Result=@n1)
  ]
 }
 Reactive_ChestOpen {
  path=AnimGraph.AnimGraphNode_StateMachine_0.Animation Testing.AnimStateNode_5.Reactive_ChestOpen
  nodes [
   n1 = Sequence Player 'PinkySight_Reactive_Gimmie_PACK'() {bLoopAnimation=false, sequence=/BRCosmetics/Accessories/FORT_Backpacks/Backpack_M_MED_PinkySight/Animations/PinkySight_Reactive_Gimmie_PACK}
   Output Animation Pose(Result=@n1)
  ]
 }
 Idle_DBNO {
  path=AnimGraph.AnimGraphNode_StateMachine_0.Animation Testing.AnimStateNode_6.Idle_DBNO
  nodes [
   n1 = Sequence Player 'PinkySight_Reactive_Sad_PACK'() {bLoopAnimation=false, sequence=&PinkySight_Reactive_Sad_PACK}
   Output Animation Pose(Result=@n1)
  ]
 }
 Transition {path=AnimGraph.AnimGraphNode_StateMachine_0.Animation Testing.AnimStateTransitionNode_0.Transition, nodes [Result()]}
 Transition {path=AnimGraph.AnimGraphNode_StateMachine_0.Animation Testing.AnimStateTransitionNode_1.Transition, nodes [Result()]}
 Transition {path=AnimGraph.AnimGraphNode_StateMachine_0.Animation Testing.AnimStateTransitionNode_2.Transition, nodes [Result()]}
 Transition {path=AnimGraph.AnimGraphNode_StateMachine_0.Animation Testing.AnimStateTransitionNode_3.Transition, nodes [Result()]}
 Transition {path=AnimGraph.AnimGraphNode_StateMachine_0.Animation Testing.AnimStateTransitionNode_4.Transition, nodes [Result()]}
 Transition {path=AnimGraph.AnimGraphNode_StateMachine_0.Animation Testing.AnimStateTransitionNode_5.Transition, nodes [Result()]}
 Transition {path=AnimGraph.AnimGraphNode_StateMachine_0.Animation Testing.AnimStateTransitionNode_7.Transition, nodes [Result()]}
 Backpack States {
  entry=Idle
  path=AnimGraph.AnimGraphNode_StateMachine_1.Backpack States
  ids=id
  nodes [
   AnimStateNode_0 = Idle() {stateEntered {guid=#1, notifyName=Entered_RandomLocoBreak}}
   AnimStateNode_3 = DBNO() {is=title AnimStateNode}
   AnimStateNode_4 = Elimination() {is=title AnimStateNode}
   AnimStateNode_5 = GainHealth or Shields() {is=title AnimStateNode}
   AnimStateNode_7 = OpenChest() {is=title AnimStateNode}
   AnimStateTransitionNode_0 = AnimStateTransitionNode() {bAllowInertializationForSelfTransitions=true}
   AnimStateTransitionNode_1 = AnimStateTransitionNode() {bAllowInertializationForSelfTransitions=true, bAutomaticRuleBasedOnSequencePlayerInState=true}
   AnimStateTransitionNode_11 = AnimStateTransitionNode()
   AnimStateTransitionNode_3 = AnimStateTransitionNode() {crossfadeDuration=0}
   AnimStateTransitionNode_4 = AnimStateTransitionNode()
   AnimStateTransitionNode_5 = AnimStateTransitionNode() {bAutomaticRuleBasedOnSequencePlayerInState=true}
   AnimStateTransitionNode_7 = AnimStateTransitionNode()
   AnimStateTransitionNode_8 = AnimStateTransitionNode() {bAutomaticRuleBasedOnSequencePlayerInState=true}
  ]
  links [
   {
    dst=Idle
    rule=AnimGraph.AnimGraphNode_StateMachine_1.Backpack States.AnimStateTransitionNode_4.AnimationTransitionGraph_0
    src=DBNO
    via=AnimStateTransitionNode_4
   }
   {
    dst=Idle
    rule=AnimGraph.AnimGraphNode_StateMachine_1.Backpack States.AnimStateTransitionNode_5.Transition
    src=Elimination
    via=AnimStateTransitionNode_5
   }
   {
    dst=Idle
    rule=AnimGraph.AnimGraphNode_StateMachine_1.Backpack States.AnimStateTransitionNode_1.Transition
    src=GainHealth or Shields
    via=AnimStateTransitionNode_1
   }
   {
    dst=DBNO
    rule=AnimGraph.AnimGraphNode_StateMachine_1.Backpack States.AnimStateTransitionNode_3.AnimationTransitionGraph_0
    src=Idle
    via=AnimStateTransitionNode_3
   }
   {
    dst=Elimination
    rule=AnimGraph.AnimGraphNode_StateMachine_1.Backpack States.AnimStateTransitionNode_11.Transition
    src=Idle
    via=AnimStateTransitionNode_11
   }
   {
    dst=GainHealth or Shields
    rule=AnimGraph.AnimGraphNode_StateMachine_1.Backpack States.AnimStateTransitionNode_0.Transition
    src=Idle
    via=AnimStateTransitionNode_0
   }
   {
    dst=OpenChest
    rule=AnimGraph.AnimGraphNode_StateMachine_1.Backpack States.AnimStateTransitionNode_7.Transition
    src=Idle
    via=AnimStateTransitionNode_7
   }
   {
    dst=Idle
    rule=AnimGraph.AnimGraphNode_StateMachine_1.Backpack States.AnimStateTransitionNode_8.Transition
    src=OpenChest
    via=AnimStateTransitionNode_8
   }
  ]
 }
 Idle {
  path=AnimGraph.AnimGraphNode_StateMachine_1.Backpack States.AnimStateNode_0.Idle
  nodes [Output Animation Pose(Result=Sequence Player 'PinkySight_Idle_PACK'() {sequence=&PinkySight_Idle_PACK})]
 }
 DBNO {
  path=AnimGraph.AnimGraphNode_StateMachine_1.Backpack States.AnimStateNode_3.DBNO
  nodes [Output Animation Pose(Result=Sequence Player 'PinkySight_Reactive_Sad_PACK'() {sequence=&PinkySight_Reactive_Sad_PACK})]
 }
 Elimination {
  path=AnimGraph.AnimGraphNode_StateMachine_1.Backpack States.AnimStateNode_4.Elimination
  nodes [
   n1 = Sequence Player 'PinkySight_Reactive_HopHop_PACK'() {bLoopAnimation=false, sequence=&PinkySight_Reactive_HopHop_PACK}
   Output Animation Pose(Result=@n1)
  ]
 }
 GainHealth or Shields {
  path=AnimGraph.AnimGraphNode_StateMachine_1.Backpack States.AnimStateNode_5.GainHealth or Shields
  nodes [
   n1 = Sequence Player 'PinkySight_Reactive_Celebrate_PACK'() {
    bLoopAnimation=false
    sequence=&PinkySight_Reactive_Celebrate_PACK
   }
   Output Animation Pose(Result=@n1)
  ]
 }
 OpenChest {
  path=AnimGraph.AnimGraphNode_StateMachine_1.Backpack States.AnimStateNode_7.OpenChest
  nodes [
   n1 = Sequence Player 'PinkySight_Reactive_Gimmie_PACK'() {bLoopAnimation=false, sequence=&PinkySight_Reactive_Gimmie_PACK}
   Output Animation Pose(Result=@n1)
  ]
 }
 Transition {
  path=AnimGraph.AnimGraphNode_StateMachine_1.Backpack States.AnimStateTransitionNode_0.Transition
  nodes [Result(bCanEnterTransition=Get GainedHealthShields() -> (GainedHealthShields:bool))]
 }
 Transition {path=AnimGraph.AnimGraphNode_StateMachine_1.Backpack States.AnimStateTransitionNode_1.Transition, nodes [Result()]}
 Transition {
  path=AnimGraph.AnimGraphNode_StateMachine_1.Backpack States.AnimStateTransitionNode_11.Transition
  nodes [Result(bCanEnterTransition=Get bEnemyPlayerEliminated() -> (bEnemyPlayerEliminated:bool))]
 }
 AnimationTransitionGraph_0 {
  path=AnimGraph.AnimGraphNode_StateMachine_1.Backpack States.AnimStateTransitionNode_3.AnimationTransitionGraph_0
  nodes [Result(bCanEnterTransition=Get bIsDBNO())]
 }
 AnimationTransitionGraph_0 {
  path=AnimGraph.AnimGraphNode_StateMachine_1.Backpack States.AnimStateTransitionNode_4.AnimationTransitionGraph_0
  nodes [Result(bCanEnterTransition=NOT Boolean(A=Get bIsDBNO()))]
 }
 Transition {path=AnimGraph.AnimGraphNode_StateMachine_1.Backpack States.AnimStateTransitionNode_5.Transition, nodes [Result()]}
 Transition {
  path=AnimGraph.AnimGraphNode_StateMachine_1.Backpack States.AnimStateTransitionNode_7.Transition
  nodes [Result(bCanEnterTransition=Get bChestOpened() -> (bChestOpened:bool))]
 }
 Transition {path=AnimGraph.AnimGraphNode_StateMachine_1.Backpack States.AnimStateTransitionNode_8.Transition, nodes [Result()]}
 EventGraph {nodes [
   Event BlueprintUpdateAnimation()
   > n1 = Sequence()
   > n2 = Random Vector Function!(
    execute
    CharacterPartAnimInstance:Custom Character Part Anim Instance Object Reference
    SkelMeshComponent:Mesh Component Object Reference=GetOwnerSkeletalMeshComponent()
    JointName:Name=neck_01
    GameTimeInSeconds:double=GetGameTimeInSeconds()
    TimeOffset:double=0
    PerlinScale:double=0
    Frequency:Vector="0.9,1.1,0.75"
    Amplitude:Vector="0.3,0.7,0.5"
    Offset:Vector="0,0,0"
   ) -> (then, VectorOut:Vector)
   > Set movement(movement:Vector=@n2) -> (Output_Get:Vector)
   Set Skydive(execute=@n1.then_1, Skydive:double=SelectFloat(A=0.2, B=1, bPickA=Get bIsSkyDiving())) -> (Output_Get:double)
   Event BlueprintPostEvaluateAnimation() {is=title K2Node_Event}
   > Set bEnemyPlayerEliminated(bEnemyPlayerEliminated:bool=false) -> (Output_Get:bool)
   > Set bChestOpened(bChestOpened:bool=false) -> (Output_Get:bool)
   n3 = Event BlueprintProcessAnimInputEvent()
   > n4 = Sequence!(execute) -> (then_0, then_1, then_2:Exec, then_3:Exec)
   > n5 = Branch(Condition=IsNamed(self=@n3.AnimInputEvent, InEventName=aie.EnemyPlayerEliminated))
   > Set bEnemyPlayerEliminated(bEnemyPlayerEliminated=true)
   Branch(execute=@n5.else, Condition=IsNamed(self=@n3.AnimInputEvent, InEventName=aie.ChestOpened))
   > Set bChestOpened(bChestOpened=true)
   Branch(
    execute=@n4.then_1
    Condition=OR Boolean(
     A=IsNamed(self=@n3.AnimInputEvent, InEventName=aie.PlayerVitality.HealthIncreased)
     B=IsNamed(self=@n3.AnimInputEvent, InEventName=aie.PlayerVitality.ShieldsIncreased)
    )
   )
   > Set GainedHealthShields(GainedHealthShields:bool=true) -> (Output_Get:bool)
   > RetriggerableDelay(Duration=0.2)
   > Set GainedHealthShields(GainedHealthShields=false)
   Get bIsDBNO()
  ]}
]
variables [
 AllowLocomotionBreak {type=boolean}
 bChestOpened {type=boolean}
 bEnemyPlayerEliminated {type=boolean}
 Current Health {type=number}
 GainedHealthShields {type=boolean}
 movement {type=Vector}
 Previous Health {type=number}
 Skydive {type=number}
]
